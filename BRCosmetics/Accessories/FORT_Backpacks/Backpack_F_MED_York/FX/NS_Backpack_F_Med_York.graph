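# Niagara system for the F_MED_York backpack cosmetic, emitting pale lilac-to-white smoke from the pack's FX_Idle socket.
# Owner speed raises spawn rate and shortens lifetime, while sprites get cone velocity, inherited socket motion, drag, gravity and SubUV animation.
# Positions blend from owner-local spawn point to simulated position, so smoke starts attached then detaches; alpha fades over life and with distance from owner.
# A socket-following sprite-plus-light emitter is disabled.
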NiagaraSystem /BRCosmetics/Accessories/FORT_Backpacks/Backpack_F_MED_York/FX/NS_Backpack_F_Med_York
dependencies {
 usedDataInterfaces [&NiagaraDataInterfaceCurve]
 usedDynamicInputs []
 usedModules [/Game/Effects/Niagara/Modules/SkeletalMesh/FN_SkeletalBoneData_v2, &SystemState]
}
emitters {
 Light {
  bEnabled=false
  emitterUpdateScript [
   &EmitterState(Life Cycle Mode=::NewEnumerator1, Loop Duration=1, UseLoopDelay=false)
   &SpawnRate(SpawnRate=20, Use Spawn Probability=false) {enabled=false}
   &SpawnBurst_Instantaneous(Spawn Count=1, Use Spawn Probability=false)
  ]
  particleSpawnScript [
   /Niagara/Modules/Spawn/Initialization/InitializeParticle(
    Color=<rgba 50 10 500 1>
    Module.Write Color=true
    Module.Write Lifetime=true
    Module.Write Mass=true
    Module.Write Material Random=true
    Module.Write Position=false
    Module.Write Scale=false
    Module.Write SpriteRotation=false
    Module.Write SpriteSize=true
    Position=&SimulationPosition()
    Sprite Size=<50 50>
   )
   &SkeletalMeshLocation(
    Inherit Velocity=false
    MaskSpawnGroup=false
    Mesh Sampling Type=::NewEnumerator5
    Skeletal Mesh=DataInterface() {propertyValues {
      DefaultMesh=null
      ExcludeBoneName=root
      OptionalComponentClass=null
      PreviewMesh=/BRCosmetics/Accessories/FORT_Backpacks/Backpack_F_MED_York/Meshes/F_MED_York_Pack
      SkinningMode=SkinOnTheFly
      SoftSourceActor=null
      SourceMode=Default
      UvSetIndex=0
      WholeMeshLOD=-1
      bExcludeBone=false
      bReadDeformedGeometry=true
      bRequireCurrentFrameData=true
      ComponentTags []
      FilteredBones []
      FilteredSockets [FX_Idle]
      MeshUserParameter {parameter {name=None, typeDefHandle {registeredTypeIndex=1}, varData []}}
      SamplingRegions []
     }}
    Socket Sampling Mode=::NewEnumerator5
   )
   /SetVariables(Particles.LightExponent=10) {isSetParameters=true}
  ]
  particleUpdateScript [
   &ParticleState()
   &SkeletalMeshLocation(
    Bone / Socket Index=1
    Inherit Velocity=false
    MaskSpawnGroup=false
    Mesh Sampling Type=::NewEnumerator5
    Sampled Position Offset=&RotateVectorByQuaternion(Vector To Rotate=<0 0 0>, Quaternion=<0 0 0 1>)
    Skeletal Mesh=DataInterface() {propertyValues {
      DefaultMesh=null
      ExcludeBoneName=root
      OptionalComponentClass=null
      PreviewMesh=&F_MED_York_Pack
      SkinningMode=SkinOnTheFly
      SoftSourceActor=null
      SourceMode=Default
      UvSetIndex=0
      WholeMeshLOD=-1
      bExcludeBone=false
      bReadDeformedGeometry=true
      bRequireCurrentFrameData=true
      ComponentTags []
      FilteredBones []
      FilteredSockets [FX_Idle,FX_Light]
      MeshUserParameter {parameter {name=None, typeDefHandle {registeredTypeIndex=1}, varData []}}
      SamplingRegions []
     }}
    Socket Sampling Mode=::NewEnumerator5
   ) {module=SkeletalMeshLocation001}
   &SolveForcesAndVelocity(Clamp Velocity=false, Limit Acceleration=false)
  ]
  rendererClasses [&NiagaraSpriteRendererProperties, &NiagaraLightRendererProperties]
  renderers [
   {rendererClass=&NiagaraSpriteRendererProperties}
   {rendererClass=&NiagaraLightRendererProperties, rendererIndex=1}
  ]
 }
 Smoke {
  emitterUpdateScript [
   &EmitterState(Life Cycle Mode=::NewEnumerator1, Loop Duration=1, UseLoopDelay=false)
   /SetVariables(Emitter.PlayerVelocityVector:Vector3f=$Engine.Owner.Velocity) {isSetParameters=true}
   /SetVariables(Emitter.PlayerSpeed:NiagaraFloat=&VectorLength(Vector=$Emitter.PlayerVelocityVector)) {isSetParameters=true}
   &SpawnRate(
    SpawnRate=&Lerp_Float(A=5, B=30, Alpha=&ClampFloat(Float=&Divide_Float(A=$Emitter.PlayerSpeed, B=10)))
    Use Spawn Probability=false
   )
  ]
  particleSpawnScript [
   &InitializeParticle(
    Color=<rgba 1 1 1 0.1>
    Lifetime=&Multiply_Float(
     A=&RandomRangeFloat(Minimum=0.75, Maximum=1)
     B=&Lerp_Float(A=2, B=0.5, Alpha=&ClampFloat(Float=&Divide_Float(A=$Emitter.PlayerSpeed, B=100)))
    )
    Module.Write Color=true
    Module.Write Lifetime=true
    Module.Write Mass=true
    Module.Write Material Random=true
    Module.Write Position=false
    Module.Write Scale=false
    Module.Write SpriteRotation=true
    Module.Write SpriteSize=true
    Position=&SimulationPosition()
    Sprite Rotation=&RandomRangeFloat(Minimum=0, Maximum=360)
    Sprite Size=&Multiply_Vector2(
     A=&Vector2DFromFloat(Value=30)
     B=&Vector2DFromFloat(Value=&RandomRangeFloat(Minimum=0.65, Maximum=1))
    )
   )
   &SkeletalMeshLocation(
    Inherit Velocity=false
    MaskSpawnGroup=false
    Mesh Sampling Type=::NewEnumerator5
    Sampled Position Offset=&RandomRangeVector(Minimum=<0 0 -4>, Maximum=<0 0 4>)
    Skeletal Mesh=DataInterface() {propertyValues {
      DefaultMesh=null
      ExcludeBoneName=root
      OptionalComponentClass=null
      PreviewMesh=&F_MED_York_Pack
      SkinningMode=SkinOnTheFly
      SoftSourceActor=null
      SourceMode=Default
      UvSetIndex=0
      WholeMeshLOD=-1
      bExcludeBone=false
      bReadDeformedGeometry=true
      bRequireCurrentFrameData=true
      ComponentTags []
      FilteredBones []
      FilteredSockets [FX_Idle]
      MeshUserParameter {parameter {name=None, typeDefHandle {registeredTypeIndex=1}, varData []}}
      SamplingRegions []
     }}
    Socket Sampling Mode=::NewEnumerator5
   )
   /SetVariables(Particles.LocalPosition:Vector3f=/Niagara/DynamicInputs/Transforms/TransformPosition(Position=Linked() {linkedVariable {name=Particles.Position, type=NiagaraPosition}}, Source Space=::World, Destination Space=::Local)) {isSetParameters=true}
   &AddVelocityInCone(
    Cone Axis=<0 -1 0>
    Module.Use Velocity Falloff On Cone Axis=false
    Velocity Strength=&Multiply_Float(A=&RandomRangeFloat(Minimum=0.25, Maximum=1), B=100)
   )
   /SetVariables(Particles.SubUV_Random:NiagaraFloat=&RandomRangeFloat(Minimum=0.5, Maximum=1)) {isSetParameters=true}
   &InheritVelocity(Module.Limit Speed=true, Velocity Scale=<3 3 3>, Velocity Source=$Output.SkeletalMeshLocation.SampledVelocity)
   &ApplyInitialForces()
   &CameraOffset(Camera Offset Amount=&RandomRangeFloat(Minimum=0, Maximum=4))
  ]
  particleUpdateScript [
   &ParticleState()
   &GravityForce(Gravity=&Multiply_Vector(A=&RandomRangeVector(Minimum=<0 0 0.25>, Maximum=<1 1 1>), B=<1 1 40>))
   &Drag(Drag=2, Use Linear Drag=true, Use Rotational Drag=false)
   &SolveForcesAndVelocity(Clamp Velocity=false, Limit Acceleration=false)
   &SubUVAnimation(End Frame:NiagaraInt32=&RandomRangeInt(Minimum=48, Maximum=63), Start Frame=24)
   &SpriteRotationRate(Rotation Rate=&Multiply_Float(A=&RandomRangeFloat(Minimum=-8, Maximum=1), B=20))
   &ScaleSpriteSize(Scale Factor=&Multiply_Vector2(
     A=<1 1>
     B=&Vector2DFromFloat(Value=&FloatFromCurve(FloatCurve=DataInterface() {propertyValues {Curve {keys [<at 0 0.3 RCIM_Cubic>,<at 1 1>]}}}))
    ))
   /SetVariables(Particles.PositionLerp:Vector3f=&Lerp_Vector(
     A=&Add_Vector(
      A=$Output.SkeletalMeshLocation.SampledPosition
      B=&RandomRangeVector(Minimum=<0 0 0>, Maximum=<0 0 0>, Evaluation Type=::NewEnumerator1)
     )
     B=Linked() {linkedVariable {name=Particles.Position, type=NiagaraPosition}}
     Alpha=&FloatFromCurve(FloatCurve=DataInterface() {propertyValues {Curve {keys [
          {arriveTangent=2.1745, interpMode=RCIM_Cubic, leaveTangent=2.1745, tangentMode=RCTM_User, time=0.75, value=0}
          {interpMode=RCIM_Cubic, leaveTangent=2.9117, time=1, value=0.2}
         ]}}})
    )) {enabled=false, isSetParameters=true}
   /SetVariables(Particles.Position=&Lerp_Vector(
     A=&TransformPosition(Position=$Particles.LocalPosition, Source Space=::Local, Destination Space=::Simulation)
     B=Linked() {linkedVariable {name=Particles.Position, type=NiagaraPosition}}
     Alpha=&FloatFromCurve(FloatCurve=DataInterface() {propertyValues {Curve {keys [<at 0 0 RCIM_Cubic>, {arriveTangent=2.7386, interpMode=RCIM_Cubic, tangentMode=RCTM_Break, time=0.6, value=1}]}}})
    )) {isSetParameters=true}
   &ScaleColor(
    Scale Alpha=&FloatFromCurve(
     FloatCurve=DataInterface() {propertyValues {Curve {keys [<at 0 0 RCIM_Cubic>,<at 0.2 1>,<at 1 0>]}}}
     Scale Curve=&FloatFromCurve(
      FloatCurve=DataInterface() {propertyValues {Curve {keys [<at 0.4467 1>,<at 1 0>]}}}
      CurveIndex=&NormalizeDistanceRange(
       Start Position:NiagaraPosition=$Engine.Owner.Position
       End Position:NiagaraPosition=$Particles.Position
       Distance:NiagaraFloat=25
       Invert Normalized Range:NiagaraBool=false
      )
     )
    )
    Scale RGB=&Lerp_Vector(
     A=&MakeVectorFromLinearColorRGB(Color=<rgba 0.8828 0.6927 1 1>)
     B=&MakeVectorFromLinearColorRGB(Color=<rgba 0.9792 0.9583 1 1>)
     Alpha=&FloatFromCurve(FloatCurve=DataInterface() {propertyValues {Curve {keys [{arriveTangent=3.1193, interpMode=RCIM_Cubic, leaveTangent=3.1193, tangentMode=RCTM_User, time=0, value=0}, <at 0.75 1>]}}})
    )
    ScaleA=true
    ScaleRGB=true
   )
  ]
  rendererClasses [&NiagaraSpriteRendererProperties]
  renderers [
   {rendererClass=&NiagaraSpriteRendererProperties}
   {rendererClass=&NiagaraSpriteRendererProperties, rendererIndex=1}
  ]
 }
}
summary {
 systemName=NS_Backpack_F_Med_York
 emitters [
  {
   bEnabled=false
   emitterName=Light
   rendererClasses [&NiagaraSpriteRendererProperties, &NiagaraLightRendererProperties]
  }
  {emitterName=Smoke, rendererClasses [&NiagaraSpriteRendererProperties]}
 ]
 userVariables [User.SkeletalMesh {
   type=NiagaraDataInterfaceSkeletalMesh
   defaultValue {
    struct=NiagaraExt_VariableValue_DataInterface
    value {dataInterface=&self:NiagaraDataInterfaceSkeletalMesh_0}
   }
  }]
}
