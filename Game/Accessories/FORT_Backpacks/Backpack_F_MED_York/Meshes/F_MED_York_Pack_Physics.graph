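# Physics asset for the F_MED_York backpack mesh, giving its left and right door bones bodies for physical swinging motion.
# Each side pairs a small box on the base bone with a larger, offset box on the parent bone, joined by one constraint.
# Both constraints lock twist and swing2, leaving swing1 as the only angular axis so each door moves like a hinge.
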PhysicsAsset /Game/Accessories/FORT_Backpacks/Backpack_F_MED_York/Meshes/F_MED_York_Pack_Physics
bodies {
 dyn_door_l_base [{shapeType=Box, extentX=0.5, extentY=0.5, extentZ=0.5}]
 dyn_door_l_parent [{shapeType=Box, extentX=1.01, extentY=4.76, extentZ=15.76, center=<1.5638 0.0207 7.796>, rotation=<rot 0.0002 -90 -0.0112>}]
 dyn_door_r_base [{shapeType=Box, extentX=0.5, extentY=0.5, extentZ=0.5}]
 dyn_door_r_parent [{shapeType=Box, extentX=1.01, extentY=4.76, extentZ=15.76, center=<1.5513 -0.0207 7.796>, rotation=<rot 0 -90 -0.0111>}]
}
constraints [
 {
  bone1Name=dyn_door_l_parent
  bone2Name=dyn_door_l_base
  swing1LimitDegrees=15
  swing2LimitDegrees=0
  swing2Motion=Locked
  twistLimitDegrees=0
  twistMotion=Locked
 }
 {
  bone1Name=dyn_door_r_parent
  bone2Name=dyn_door_r_base
  swing1LimitDegrees=15
  swing2LimitDegrees=0
  swing2Motion=Locked
  twistLimitDegrees=0
  twistMotion=Locked
 }
]
