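# Animation Blueprint for the F_MED_York backpack, built to swing its two doors while jumping or falling and otherwise simulate physics.
# Sequence logic sets opposing alphas from jump/fall flags for interpolated additive door-bone rotations and a RigidBody node using the pack's physics asset.
# As wired, the update event does not drive that logic, and the output passes the input pose through, bypassing both door and physics nodes.
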AnimBlueprint /BRCosmetics/Accessories/FORT_Backpacks/Backpack_F_MED_York/Meshes/F_MED_York_Pack_AnimBP
target_skeleton=/Game/Accessories/FORT_Backpacks/Backpack_F_MED_York/Meshes/F_MED_York_Pack_Skeleton
graphs [
 AnimGraph {nodes [
   n1 = Get Jump_Alpha() -> (Jump_Alpha:double)
   n2 = "Transform (Modify) Bone - Bone: dyn_door_l_parent"(Alpha=@n1, Translation="0,0,0", Rotation="0,-30,0", Scale="1,1,1") {
    rotationMode=BMM_Additive
    rotationSpace=BCS_BoneSpace
    alphaBoolBlend {blendOption=Linear}
    alphaScaleBiasClamp {bInterpResult=true, interpSpeedDecreasing=2, interpSpeedIncreasing=8}
    boneToModify {boneName=dyn_door_l_parent}
    rotation {yaw=-30}
   }
   "Transform (Modify) Bone - Bone: dyn_door_r_parent"(ComponentPose=@n2, Alpha=@n1, Translation="0,0,0", Rotation="0,30,0", Scale="1,1,1") {
    rotationMode=BMM_Additive
    rotationSpace=BCS_BoneSpace
    alphaBoolBlend {blendOption=Linear}
    alphaScaleBiasClamp {bInterpResult=true, interpSpeedDecreasing=2, interpSpeedIncreasing=8}
    boneToModify {boneName=dyn_door_r_parent}
    rotation {yaw=30}
   }
   RigidBody(Alpha=Get InverseJump_Alpha() -> (InverseJump_Alpha:double)) {
    bClampLinearTranslationLimitToRefPose=true
    bOverrideWorldGravity=true
    overridePhysicsAsset=/Game/Accessories/FORT_Backpacks/Backpack_F_MED_York/Meshes/F_MED_York_Pack_Physics
    simulationSpace=BaseBoneSpace
    alphaBoolBlend {blendOption=Linear}
    alphaScaleBiasClamp {bInterpResult=true, interpSpeedDecreasing=7, interpSpeedIncreasing=8}
    baseBoneRef {boneName=root}
    componentLinearAccScale=<0.3 0.25 0.25>
    componentLinearVelScale=<0.3 0.25 0.25>
    overrideWorldGravity {z=-980}
   }
   n3 = Input Pose()
   Output Pose(Result=Component To Local(ComponentPose=Local To Component(LocalPose=@n3))) {layerGroup=None}
  ]}
 EventGraph {nodes [
   n1 = Sequence()
   n2 = Get bIsJumping()
   n3 = Get bIsFalling()
   > Set Jump_Alpha(Jump_Alpha:double=SelectFloat(A=1, B=0, bPickA=OR Boolean(A=@n2, B=@n3))) -> (Output_Get:double)
   n4 = Get bIsJumping()
   n5 = Get bIsFalling()
   Set InverseJump_Alpha(execute=@n1.then_1, InverseJump_Alpha:double=SelectFloat(A=0, B=1, bPickA=OR Boolean(A=@n4, B=@n5))) -> (Output_Get:double)
   TryGetPawnOwner()
   Event BlueprintUpdateAnimation()
  ]}
]
variables [
 InverseJump_Alpha {type=number}
 Jump_Alpha {type=number}
]
